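; Passive quest ability for the Season 21 Battle Pass "TakeSeedToThree" quest that detects a tagged item in the owning player's inventory.
; On activation it starts a looping one-second timer that reads the inventory items from the owner's player controller.
; Each item whose definition carries the configured Tag triggers applying the configured Gameplay Effect to that controller, repeating on every poll.
; A failed passive activation ends the ability.
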(event Ability|EventFailedToActivatePassiveAbility
  (params (ActorInfo :type Gameplay%20Ability%20Actor%20Info%20Structure))
  (Ability|EndAbility))

(event Ability|EventActivateAbility
  (|CallParentFunction)
  (Utilities|Time|SetTimerbyEvent (delegate Custom|CustomEvent) 1.0 true))

(event Custom|CustomEvent
  (bind _items (InventoryManagement|GetItemsfromInventoryOwner (Fortnite|GetFortPlayerControllerfromActor (Variables|GAVars|Data|GetOwningPlayerPawn))))
  (if (Utilities|Array|IsNotEmpty _items)
    (for _array_element _items
      (if (GameplayTags|HasTag (GameplayTags|GetOwnedGameplayTags (Item|GetItemDefinition _array_element) :!class /Script/GameplayTags.GameplayTagAssetInterface) (Variables|Default|GetTag) true)
        (Creative|GameplayEffects|SimpleApplyGameplayEffectToActor (Variables|Default|GetGameplay%20Effect) (Fortnite|GetFortPlayerControllerfromActor (Variables|GAVars|Data|GetOwningPlayerPawn)))))))
